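# Physics asset for the string on the M_MED_BlobRock male medium player body cosmetic, defining bodies and joints for its secondary motion.
# Capsule bodies form a branching chain: string_m hangs off string_m_MSTR and splits into a three-bone left strand and a two-bone right strand.
# Strand joints share the same swing and twist limits, while the root joint is tighter in one swing axis and twist.
# A large pelvis capsule has no constraints, apparently serving as a collider for the string.
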PhysicsAsset /BRCosmetics/Characters/Player/Male/Medium/Bodies/M_MED_BlobRock/Meshes/M_MED_BlobRock_Physics_String
bodies {
 pelvis [{shapeType=Capsule, length=7.486, radius=14.8615, shapeName=pelvis_capsule, center=<0.2079 -2.2677 0>, rotation=<rot 90 -0.2147 0>}]
 string_l_1 [{shapeType=Capsule, length=0.7995, radius=1.2005, shapeName=string_l_1_capsule, center=<1.0788 -0.2418 0>, rotation=<rot 90 0 0>}]
 string_l_2 [{shapeType=Capsule, length=0.2994, radius=1.7006, shapeName=string_l_2_capsule, center=<1.0172 -0.0356 -0.3492>, rotation=<rot 80 0 0>}]
 string_l_3 [{shapeType=Capsule, length=0.6128, radius=2.1226, shapeName=string_l_3_capsule, center=<1.6653 0.1138 0>, rotation=<rot 90 0 0>}]
 string_m [{shapeType=Capsule, length=1, radius=1, shapeName=string_m_capsule}]
 string_m_MSTR [{shapeType=Capsule, length=0.505, radius=0.505, shapeName=string_m_MSTR_capsule, rotation=<rot 0 -90 0>}]
 string_r_1 [{shapeType=Capsule, length=0.1667, radius=1.8333, shapeName=string_r_1_capsule, center=<1.0788 -0.7734 0.0809>, rotation=<rot 90 0 0>}]
 string_r_2 [{shapeType=Capsule, length=1.0557, radius=1.8646, shapeName=string_r_2_capsule, center=<1.6699 -1.027 -0.011>, rotation=<rot 90 0 0>}]
}
constraints [
 {bone1Name=string_r_2, bone2Name=string_r_1, swing1LimitDegrees=30, swing2LimitDegrees=30, twistLimitDegrees=10}
 {bone1Name=string_r_1, bone2Name=string_m, swing1LimitDegrees=30, swing2LimitDegrees=30, twistLimitDegrees=10}
 {bone1Name=string_l_3, bone2Name=string_l_2, swing1LimitDegrees=30, swing2LimitDegrees=30, twistLimitDegrees=10}
 {bone1Name=string_l_2, bone2Name=string_l_1, swing1LimitDegrees=30, swing2LimitDegrees=30, twistLimitDegrees=10}
 {bone1Name=string_l_1, bone2Name=string_m, swing1LimitDegrees=30, swing2LimitDegrees=30, twistLimitDegrees=10}
 {bone1Name=string_m, bone2Name=string_m_MSTR, swing1LimitDegrees=30, swing2LimitDegrees=5, twistLimitDegrees=5}
]
