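# Physics asset for the top piece of the male medium BlobRock body cosmetic, defining four capsule bone chains at left/right front and back.
# Tight swing and twist limits link each chain down from its MSTR root bone.
# Looser constraints tie each front tip to the back chain's end on the same side, while angularly free ones link both front tips to each other and spine_01.
# Larger capsules cover spine_01 and both thighs.
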PhysicsAsset /BRCosmetics/Characters/Player/Male/Medium/Bodies/M_MED_BlobRock/Meshes/M_MED_BlobRock_Physics_Top
bodies {
 MSTR_top_l_bk [{shapeType=Capsule, length=1, radius=1, shapeName=MSTR_top_l_bk_capsule}]
 MSTR_top_l_fr [{shapeType=Capsule, length=1, radius=1, shapeName=MSTR_top_l_fr_capsule}]
 MSTR_top_r_bk [{shapeType=Capsule, length=1, radius=1, shapeName=MSTR_top_r_bk_capsule}]
 MSTR_top_r_fr [{shapeType=Capsule, length=1, radius=1, shapeName=MSTR_top_r_fr_capsule}]
 spine_01 [{shapeType=Capsule, length=3.8102, radius=13.7506, shapeName=spine_01_capsule, center=<4.2734 0 0>, rotation=<rot 90 0 0>}]
 thigh_l [{shapeType=Capsule, length=25.3668, radius=7.8421, shapeName=thigh_l_capsule, center=<-16.2646 2.9933 0>, rotation=<rot 90 0 0>}]
 thigh_r [{shapeType=Capsule, length=23.3516, radius=9.0519, shapeName=thigh_r_capsule, center=<15.3826 -3.2413 0>, rotation=<rot 90 0 0>}]
 top_l_bk_1 [{shapeType=Capsule, length=10.6111, radius=2, shapeName=top_l_bk_1_capsule, center=<8.3591 0 0>, rotation=<rot 90 0 0>}]
 top_l_bk_2 [{shapeType=Capsule, length=7.0302, radius=2, shapeName=top_l_bk_2_capsule, center=<5.7911 0 0>, rotation=<rot 90 0 0>}]
 top_l_fr_1 [{shapeType=Capsule, length=12.4786, radius=2, shapeName=top_l_fr_1_capsule, center=<8.9035 0 0>, rotation=<rot 90 0 0>}]
 top_l_fr_2 [{shapeType=Capsule, length=9.199, radius=2, shapeName=top_l_fr_2_capsule, center=<6.5121 0 0>, rotation=<rot 90 0 0>}]
 top_l_fr_3 [{shapeType=Capsule, length=1, radius=1, shapeName=top_l_fr_3_capsule}]
 top_r_bk_1 [{shapeType=Capsule, length=10.4069, radius=2, shapeName=top_r_bk_1_capsule, center=<7.6207 0 0>, rotation=<rot 90 0 0>}]
 top_r_bk_2 [{shapeType=Capsule, length=6.3121, radius=2, shapeName=top_r_bk_2_capsule, center=<4.9934 0 0>, rotation=<rot 90 0 0>}]
 top_r_fr_1 [{shapeType=Capsule, length=12.8738, radius=2, shapeName=top_r_fr_1_capsule, center=<8.9035 0 0>, rotation=<rot 90 0 0>}]
 top_r_fr_2 [{shapeType=Capsule, length=8.0431, radius=2, shapeName=top_r_fr_2_capsule, center=<6.8816 0 0>, rotation=<rot 90 0 0>}]
 top_r_fr_3 [{shapeType=Capsule, length=1, radius=1, shapeName=top_r_fr_3_capsule}]
}
constraints [
 {bone1Name=top_l_bk_2, bone2Name=top_l_bk_1, swing1LimitDegrees=20, swing2LimitDegrees=10, twistLimitDegrees=5}
 {bone1Name=top_l_bk_1, bone2Name=MSTR_top_l_bk, swing1LimitDegrees=20, swing2LimitDegrees=3, twistLimitDegrees=3}
 {bone1Name=top_l_fr_3, bone2Name=top_l_fr_2, swing1LimitDegrees=20, swing2LimitDegrees=20, twistLimitDegrees=5}
 {bone1Name=top_l_fr_2, bone2Name=top_l_fr_1, swing1LimitDegrees=20, swing2LimitDegrees=10, twistLimitDegrees=5}
 {bone1Name=top_l_fr_1, bone2Name=MSTR_top_l_fr, swing1LimitDegrees=20, swing2LimitDegrees=3, twistLimitDegrees=3}
 {bone1Name=top_r_bk_2, bone2Name=top_r_bk_1, swing1LimitDegrees=20, swing2LimitDegrees=10, twistLimitDegrees=5}
 {bone1Name=top_r_bk_1, bone2Name=MSTR_top_r_bk, swing1LimitDegrees=20, swing2LimitDegrees=3, twistLimitDegrees=3}
 {bone1Name=top_r_fr_3, bone2Name=top_r_fr_2, swing1LimitDegrees=20, swing2LimitDegrees=20, twistLimitDegrees=5}
 {bone1Name=top_r_fr_2, bone2Name=top_r_fr_1, swing1LimitDegrees=20, swing2LimitDegrees=10, twistLimitDegrees=5}
 {bone1Name=top_r_fr_1, bone2Name=MSTR_top_r_fr, swing1LimitDegrees=20, swing2LimitDegrees=3, twistLimitDegrees=3}
 {bone1Name=top_l_fr_3, bone2Name=top_l_bk_2, swing1LimitDegrees=45, swing2LimitDegrees=45, twistLimitDegrees=5}
 {bone1Name=top_r_fr_3, bone2Name=top_r_bk_2, swing1LimitDegrees=45, swing2LimitDegrees=45, twistLimitDegrees=5}
 {
  bone1Name=top_r_fr_3
  bone2Name=top_l_fr_3
  swing1LimitDegrees=45
  swing1Motion=Free
  swing2LimitDegrees=45
  swing2Motion=Free
  twistLimitDegrees=45
  twistMotion=Free
 }
 {
  bone1Name=top_l_fr_3
  bone2Name=spine_01
  swing1LimitDegrees=45
  swing1Motion=Free
  swing2LimitDegrees=45
  swing2Motion=Free
  twistLimitDegrees=45
  twistMotion=Free
 }
 {
  bone1Name=top_r_fr_3
  bone2Name=spine_01
  swing1LimitDegrees=45
  swing1Motion=Free
  swing2LimitDegrees=45
  swing2Motion=Free
  twistLimitDegrees=45
  twistMotion=Free
 }
]
