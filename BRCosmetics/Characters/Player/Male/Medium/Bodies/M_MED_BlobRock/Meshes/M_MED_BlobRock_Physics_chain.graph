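# Physics asset for the chain on the M_MED_BlobRock player body cosmetic, set up for physical chain motion and collision.
# Small capsules form left and right strands from MSTR_chain anchors, linked by uniformly limited swing and tight-twist constraints.
# A constraint between chain_3_l and chain_4_r joins the strands into one span between both anchors.
# Larger unconstrained capsules on neck, clavicles, upper spine and left upper arm approximate the body as collision shapes.
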PhysicsAsset /BRCosmetics/Characters/Player/Male/Medium/Bodies/M_MED_BlobRock/Meshes/M_MED_BlobRock_Physics_chain
bodies {
 MSTR_chain_l [{shapeType=Capsule, length=0.505, radius=0.505, shapeName=MSTR_chain_l_capsule, rotation=<rot 0 -90 0>}]
 MSTR_chain_r [{shapeType=Capsule, length=0.505, radius=0.505, shapeName=MSTR_chain_r_capsule, rotation=<rot 0 -90 0>}]
 chain_1_l [{
   shapeType=Capsule
   length=5.5191
   radius=0.8502
   shapeName=chain_1_l_capsule
   center=<3.6156 -0.1564 0.0112>
   rotation=<rot -0.0418 89.5358 -89.3408>
  }]
 chain_1_r [{
   shapeType=Capsule
   length=5.0334
   radius=0.7504
   shapeName=chain_1_r_capsule
   center=<0.9217 0.1112 0.0293>
   rotation=<rot -0.0406 -91.3183 -89.194>
  }]
 chain_2_l [{
   shapeType=Capsule
   length=3.9632
   radius=0.8669
   shapeName=chain_2_l_capsule
   center=<2.8247 -0.2304 -0.046>
   rotation=<rot -0.0859 -91.0562 -89.3013>
  }]
 chain_2_r [{
   shapeType=Capsule
   length=4.4022
   radius=0.8968
   shapeName=chain_2_r_capsule
   center=<3.0688 -0.1993 0.0503>
   rotation=<rot -0.0536 91.7471 -89.4539>
  }]
 chain_3_l [{
   shapeType=Capsule
   length=4.4085
   radius=1.0501
   shapeName=chain_3_l_capsule
   center=<3.0011 -0.5153 0.067>
   rotation=<rot -0.7028 -93.7359 -91.108>
  }]
 chain_3_r [{
   shapeType=Capsule
   length=4.447
   radius=0.9661
   shapeName=chain_3_r_capsule
   center=<3.0329 -0.5537 -0.0159>
   rotation=<rot -0.0765 87.5904 -89.5959>
  }]
 chain_4_r [{
   shapeType=Capsule
   length=2.6263
   radius=0.9599
   shapeName=chain_4_r_capsule
   center=<-0.0023 0.1818 0.0085>
   rotation=<rot 0 89.7952 -87.3536>
  }]
 clavicle_l_dyn [{
   shapeType=Capsule
   length=2.9303
   radius=5.331
   shapeName=clavicle_l_dyn_capsule
   center=<-1.9675 2.3399 -5.0853>
   rotation=<rot -0.0001 -90 -81.0936>
  }]
 clavicle_r_dyn [{
   shapeType=Capsule
   length=3.9185
   radius=4.5695
   shapeName=clavicle_r_dyn_capsule
   center=<0.8049 -3.1497 6.4324>
   rotation=<rot -73.4478 179.9997 -179.9997>
  }]
 neck_01 [{shapeType=Capsule, length=6.4161, radius=8.9976, shapeName=neck_01_capsule, center=<2.4199 3.4338 0>, rotation=<rot -90 -14.3602 0>}]
 spine_05_dyn [
  {shapeType=Capsule, length=28.6181, radius=25.2655, shapeName=spine_04_capsule, center=<-12.274 -5.1776 0>}
  {
   shapeType=Capsule
   length=3.7397
   radius=6.7547
   shapeName=spine_05_dyn_capsule
   center=<6.2223 -0.8406 -13.293>
   rotation=<rot -20.8384 -0.0001 0>
  }
  {
   shapeType=Capsule
   length=4.0003
   radius=7.5427
   shapeName=spine_05_dyn_capsule
   center=<5.4929 -1.4014 12.878>
   rotation=<rot 20.7557 8.0826 2.8812>
  }
 ]
 upperarm_twist_01_l [{shapeType=Capsule, length=5.2915, radius=7.2129, shapeName=upperarm_twist_01_l_capsule, center=<8.12 -0.499 0>, rotation=<rot 90 0 0>}]
}
constraints [
 {bone1Name=chain_1_l, bone2Name=MSTR_chain_l, swing1LimitDegrees=30, swing2LimitDegrees=30, twistLimitDegrees=5}
 {bone1Name=chain_2_l, bone2Name=chain_1_l, swing1LimitDegrees=30, swing2LimitDegrees=30, twistLimitDegrees=5}
 {bone1Name=chain_3_l, bone2Name=chain_2_l, swing1LimitDegrees=30, swing2LimitDegrees=30, twistLimitDegrees=5}
 {bone1Name=chain_1_r, bone2Name=MSTR_chain_r, swing1LimitDegrees=30, swing2LimitDegrees=30, twistLimitDegrees=5}
 {bone1Name=chain_2_r, bone2Name=chain_1_r, swing1LimitDegrees=30, swing2LimitDegrees=30, twistLimitDegrees=5}
 {bone1Name=chain_3_r, bone2Name=chain_2_r, swing1LimitDegrees=30, swing2LimitDegrees=30, twistLimitDegrees=5}
 {bone1Name=chain_4_r, bone2Name=chain_3_r, swing1LimitDegrees=30, swing2LimitDegrees=30, twistLimitDegrees=5}
 {bone1Name=chain_3_l, bone2Name=chain_4_r, swing1LimitDegrees=30, swing2LimitDegrees=30, twistLimitDegrees=5}
]
