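# Control Rig for the BlobRock male character's face accessory mesh that poses the three back hair chains as the head moves.
# Three spherical pose readers measure head orientation relative to spine_05, one on the X axis and two on Y with opposed rotation offsets.
# Each reader's output weights fixed offsets on the left, middle and right hair master bones, propagating to their children.
# An identity head offset runs first; top hair bones are untouched.
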ControlRigBlueprint /BRCosmetics/Characters/Player/Male/Medium/Bodies/M_MED_BlobRock/Meshes/Parts/M_MED_BlobRock_FaceAcc_CtrlRig
graphs [
 RigVMFunctionLibrary {links [], nodes []}
 RigVMFunction_Sequence_1_ContainedGraph {nodes [
   n1 = RigVMFunctionEntryNode()
   RigVMFunction_Sequence = &RigVMFunction_Sequence(ExecuteContext=@n1.ExecuteContext)
   n2 = &RigVMFunction_Sequence(ExecuteContext=@RigVMFunction_Sequence.B)
   RigVMFunctionReturnNode(A=@RigVMFunction_Sequence.A, B=@n2.A, C=@n2.B)
  ]}
 RigVMModel {nodes [
   RigUnit_BeginExecution = &RigUnit_BeginExecution()
   &RigUnit_OffsetTransformForItem(
    Item=(Type=Bone,Name="head")
    OffsetTransform=(Rotation=(X=0,Y=0,Z=0,W=1),Translation=(X=0,Y=0,Z=0),Scale3D=(X=1,Y=1,Z=1))
    Weight=1
    bPropagateToChildren=True
    ExecutePin=@RigUnit_BeginExecution.ExecutePin
   )
   > RigVMFunction_Sequence = RigVMAggregateNode()
   n1 = &RigUnit_SphericalPoseReader(
    ActiveRegionScaleFactors=(PositiveWidth=1,NegativeWidth=1,PositiveHeight=1.5,NegativeHeight=1.5)
    ActiveRegionSize=0.15
    Debug=(bDrawDebug=false,bDraw2D=False,bDrawLocalAxes=False,DebugScale=20,DebugSegments=20,DebugThickness=0.1)
    DriverAxis=(X=1,Y=0,Z=0)
    DriverItem=(Type=Bone,Name="head")
    FalloffRegionScaleFactors=(PositiveWidth=1,NegativeWidth=1,PositiveHeight=1,NegativeHeight=1)
    FalloffSize=0.5
    FlipHeightScaling=False
    FlipWidthScaling=False
    OptionalParentItem=(Type=Bone,Name="spine_05")
    RotationOffset=(X=-90,Y=0,Z=0)
    ExecutePin=@RigVMFunction_Sequence.A
   )
   > RigUnit_OffsetTransformForItem = &RigUnit_OffsetTransformForItem(
    Item=(Type=Bone,Name="hair_back_r_MSTR")
    OffsetTransform=(Rotation=(X=-0.2932,Y=-0.5939,Z=-0.246,W=0.7077),Translation=(X=-1,Y=0,Z=2),Scale3D=(X=1,Y=1,Z=1))
    bPropagateToChildren=True
    Weight=@n1.OutputParam
   )
   > &RigUnit_OffsetTransformForItem(
    Item=(Type=Bone,Name="hair_back_mid_MSTR")
    OffsetTransform=(Rotation=(X=0,Y=-0.6428,Z=0,W=0.766),Translation=(X=-1,Y=0,Z=2),Scale3D=(X=1,Y=1,Z=1))
    bPropagateToChildren=True
    Weight=@n1.OutputParam
   )
   > &RigUnit_OffsetTransformForItem(
    Item=(Type=Bone,Name="hair_back_l_MSTR")
    OffsetTransform=(Rotation=(X=0,Y=-0.6428,Z=0,W=0.766),Translation=(X=-1,Y=0,Z=2),Scale3D=(X=1,Y=1,Z=1))
    bPropagateToChildren=True
    Weight=@n1.OutputParam
   )
   n2 = &RigUnit_SphericalPoseReader(
    ActiveRegionScaleFactors=(PositiveWidth=1,NegativeWidth=1,PositiveHeight=1,NegativeHeight=1)
    ActiveRegionSize=0.15
    Debug=(bDrawDebug=false,bDraw2D=False,bDrawLocalAxes=False,DebugScale=20,DebugSegments=20,DebugThickness=0.1)
    DriverAxis=(X=0,Y=1,Z=0)
    DriverItem=(Type=Bone,Name="head")
    FalloffRegionScaleFactors=(PositiveWidth=1,NegativeWidth=1,PositiveHeight=1,NegativeHeight=1)
    FalloffSize=0.5
    FlipHeightScaling=False
    FlipWidthScaling=False
    OptionalParentItem=(Type=Bone,Name="spine_05")
    RotationOffset=(X=0,Y=0,Z=0)
    ExecutePin=@RigVMFunction_Sequence.B
   )
   > &RigUnit_OffsetTransformForItem(
    Item=(Type=Bone,Name="hair_back_r_MSTR")
    OffsetTransform=(Rotation=(X=0,Y=0,Z=-0.2164,W=0.9763),Translation=(X=0,Y=0,Z=0),Scale3D=(X=1,Y=1,Z=1))
    bPropagateToChildren=True
    Weight=@n2.OutputParam
   )
   > &RigUnit_OffsetTransformForItem(
    Item=(Type=Bone,Name="hair_back_mid_MSTR")
    OffsetTransform=(Rotation=(X=0,Y=0,Z=-0.2164,W=0.9763),Translation=(X=0,Y=0,Z=0),Scale3D=(X=1,Y=1,Z=1))
    bPropagateToChildren=True
    Weight=@n2.OutputParam
   )
   > &RigUnit_OffsetTransformForItem(
    Item=(Type=Bone,Name="hair_back_l_MSTR")
    OffsetTransform=(Rotation=(X=0,Y=0,Z=-0.2164,W=0.9763),Translation=(X=0,Y=0,Z=0),Scale3D=(X=1,Y=1,Z=1))
    bPropagateToChildren=True
    Weight=@n2.OutputParam
   )
   > n3 = &RigUnit_SphericalPoseReader(
    ActiveRegionScaleFactors=(PositiveWidth=1,NegativeWidth=1,PositiveHeight=1,NegativeHeight=1)
    ActiveRegionSize=0.15
    Debug=(bDrawDebug=false,bDraw2D=False,bDrawLocalAxes=False,DebugScale=20,DebugSegments=20,DebugThickness=0.1)
    DriverAxis=(X=0,Y=1,Z=0)
    DriverItem=(Type=Bone,Name="head")
    FalloffRegionScaleFactors=(PositiveWidth=1,NegativeWidth=1,PositiveHeight=1,NegativeHeight=1)
    FalloffSize=0.5
    FlipHeightScaling=False
    FlipWidthScaling=False
    OptionalParentItem=(Type=Bone,Name="spine_05")
    RotationOffset=(X=180,Y=0,Z=0)
   )
   > &RigUnit_OffsetTransformForItem(
    Item=(Type=Bone,Name="hair_back_r_MSTR")
    OffsetTransform=(Rotation=(X=0,Y=0,Z=0.342,W=0.9397),Translation=(X=0,Y=0,Z=0),Scale3D=(X=1,Y=1,Z=1))
    bPropagateToChildren=True
    Weight=@n3.OutputParam
   )
   > &RigUnit_OffsetTransformForItem(
    Item=(Type=Bone,Name="hair_back_mid_MSTR")
    OffsetTransform=(Rotation=(X=0,Y=0,Z=0.342,W=0.9397),Translation=(X=0,Y=0,Z=0),Scale3D=(X=1,Y=1,Z=1))
    bPropagateToChildren=True
    Weight=@n3.OutputParam
   )
   > &RigUnit_OffsetTransformForItem(
    Item=(Type=Bone,Name="hair_back_l_MSTR")
    OffsetTransform=(Rotation=(X=0,Y=0,Z=0.2164,W=0.9763),Translation=(X=0,Y=0,Z=0),Scale3D=(X=1,Y=1,Z=1))
    bPropagateToChildren=True
    Weight=@n3.OutputParam
   )
  ]}
]
hierarchy [
 root {loc=<0 0 0>, rot=<0 0 0 1>}
 pelvis {parent=root, loc=<0 -0.7071 93.1191>, rot=<0.0127 0.707 -0.0127 -0.707>}
 spine_01 {parent=pelvis, loc=<6.7494 -0.6388 0>, rot=<0 0 0.034 -0.9994>}
 spine_02 {parent=spine_01, loc=<7.2518 -0.035 0>, rot=<0 0 0.0238 -0.9997>}
 spine_03 {parent=spine_02, loc=<6.0104 0 0>, rot=<0 0 0.0149 -0.9999>}
 spine_04 {parent=spine_03, loc=<8.1024 0 0>, rot=<0 0 -0.0132 -0.9999>}
 spine_05 {parent=spine_04, loc=<10.0146 0 0>, rot=<0 0 -0.0015 -1>}
 neck_01 {parent=spine_05, loc=<11.3138 0.1529 0>, rot=<0 0 -0.2003 -0.9797>}
 neck_02 {parent=neck_01, loc=<4.5627 0 0>, rot=<0 0 0.0504 -0.9987>}
 head {parent=neck_02, loc=<5.3914 0 0>, rot=<0 0 0.115 -0.9934>}
 hair_top_A_base {parent=head, loc=<10.8051 6.4648 -5.1563>, rot=<0.4094 -0.1144 -0.2243 -0.8769>}
 hair_top_A_1 {parent=hair_top_A_base, loc=<0 0 0>, rot=<0 0 0 1>}
 hair_top_A_2 {parent=hair_top_A_1, loc=<4.1374 0 0>, rot=<0 0 0 1>}
 hair_top_B_base {parent=head, loc=<10.7978 7.3581 4.3399>, rot=<0.347 0.2059 -0.145 -0.9034>}
 hair_top_B_1 {parent=hair_top_B_base, loc=<0 0 0>, rot=<0 0 0 1>}
 hair_top_B_2 {parent=hair_top_B_1, loc=<3.9808 0 0>, rot=<0 0 0 1>}
 hair_top_C_base {parent=head, loc=<10.8663 -0.938 5.1933>, rot=<0.3499 0.1259 0.2954 -0.88>}
 hair_top_C_1 {parent=hair_top_C_base, loc=<0 0 0>, rot=<0 0 0 1>}
 hair_top_C_2 {parent=hair_top_C_1, loc=<4.468 0 0>, rot=<0 0 0 1>}
 hair_top_D_base {parent=head, loc=<10.8746 -1.9319 -3.8307>, rot=<0.3941 -0.3433 0.2012 -0.8285>}
 hair_top_D_1 {parent=hair_top_D_base, loc=<0 0 0>, rot=<0 0 0 1>}
 hair_top_D_2 {parent=hair_top_D_1, loc=<4.468 0 0>, rot=<0 0 0 1>}
 hair_back_mid_MSTR {parent=head, loc=<3.4796 -3.7389 0>, rot=<0.2204 -0.6719 0.6719 -0.2204>}
 hair_back_mid_1 {parent=hair_back_mid_MSTR, loc=<0 0 0>, rot=<0 0 0 1>}
 hair_back_mid_2 {parent=hair_back_mid_1, loc=<5.9344 0 0>, rot=<0 0 0 1>}
 hair_back_mid_3 {parent=hair_back_mid_2, loc=<5.7 0 0>, rot=<0 0 0 1>}
 hair_back_l_MSTR {parent=head, loc=<3.7462 -3.543 -4.1244>, rot=<0.0582 -0.7779 0.5469 -0.304>}
 hair_back_l_1 {parent=hair_back_l_MSTR, loc=<0 0 0>, rot=<0 0 0 1>}
 hair_back_l_2 {parent=hair_back_l_1, loc=<5.2 0 0>, rot=<-0.0163 0.1437 0.1114 -0.9832>}
 hair_back_l_3 {parent=hair_back_l_2, loc=<5.9182 0 0>, rot=<0 0 -0.0878 -0.9961>}
 hair_back_r_MSTR {parent=head, loc=<3.7462 -3.543 4.1244>, rot=<0.2933 -0.5472 0.7823 -0.0495>}
 hair_back_r_1 {parent=hair_back_r_MSTR, loc=<0 0 0>, rot=<0 0 0 1>}
 hair_back_r_2 {parent=hair_back_r_1, loc=<5.45 0 0>, rot=<0.0018 0.1177 -0.0155 -0.9929>}
 hair_back_r_3 {parent=hair_back_r_2, loc=<5.2619 0 0>, rot=<0 0 0.1255 -0.9921>}
 hair_top_E_base {parent=head, loc=<11.7544 4.4997 0.4547>, rot=<0.5239 0.0109 -0.0479 -0.8504>}
 hair_top_E_1 {parent=hair_top_E_base, loc=<0 0 0>, rot=<0 0 0 1>}
 clavicle_l {parent=spine_05, loc=<4.3575 -0.1466 -2.3223>, rot=<0.0465 -0.7289 -0.009 -0.6829>}
 upperarm_l {parent=clavicle_l, loc=<15.5111 0 0>, rot=<0.0226 -0.3234 -0.0088 -0.9459>}
 clavicle_r {parent=spine_05, loc=<4.3575 -0.1466 2.3223>, rot=<0.7289 0.0465 -0.6829 0.009>}
 upperarm_r {parent=clavicle_r, loc=<-15.5111 0 0>, rot=<0.0226 -0.3234 -0.0088 -0.9459>}
]
