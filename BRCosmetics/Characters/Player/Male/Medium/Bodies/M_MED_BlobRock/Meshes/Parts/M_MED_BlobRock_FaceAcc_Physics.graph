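# Physics asset for the face accessory part of the male medium BlobRock player cosmetic, defining bodies for its top hair bones.
# Five separate chains, A to E, each start at a small base capsule and continue through one capsule in E or two in A to D.
# Constraints join only neighboring bones within each chain, sharing the same narrow swing and tighter twist limit angles, with no links between chains or to a head body.
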PhysicsAsset /BRCosmetics/Characters/Player/Male/Medium/Bodies/M_MED_BlobRock/Meshes/Parts/M_MED_BlobRock_FaceAcc_Physics
bodies {
 hair_top_A_1 [{shapeType=Capsule, length=1, radius=1, shapeName=hair_top_A_1_capsule, center=<2.4037 0 0>}]
 hair_top_A_2 [{shapeType=Capsule, length=1, radius=1, shapeName=hair_top_A_2_capsule, center=<2.4037 0 0>}]
 hair_top_A_base [{shapeType=Capsule, length=0.505, radius=0.505, shapeName=hair_top_A_base_capsule, rotation=<rot 0 -90 0>}]
 hair_top_B_1 [{shapeType=Capsule, length=1, radius=1, shapeName=hair_top_B_1_capsule, center=<2.4037 0 0>}]
 hair_top_B_2 [{shapeType=Capsule, length=1, radius=1, shapeName=hair_top_B_2_capsule, center=<2.4037 0 0>}]
 hair_top_B_base [{shapeType=Capsule, length=0.505, radius=0.505, shapeName=hair_top_B_base_capsule, rotation=<rot 0 -90 0>}]
 hair_top_C_1 [{shapeType=Capsule, length=1, radius=1, shapeName=hair_top_C_1_capsule, center=<2.4037 0 0>}]
 hair_top_C_2 [{shapeType=Capsule, length=1, radius=1, shapeName=hair_top_C_2_capsule, center=<2.4037 0 0>}]
 hair_top_C_base [{shapeType=Capsule, length=0.505, radius=0.505, shapeName=hair_top_C_base_capsule, rotation=<rot 0 -90 0>}]
 hair_top_D_1 [{shapeType=Capsule, length=1, radius=1, shapeName=hair_top_D_1_capsule, center=<2.3924 0 0>}]
 hair_top_D_2 [{shapeType=Capsule, length=1, radius=1, shapeName=hair_top_D_2_capsule, center=<2.4037 0 0>}]
 hair_top_D_base [{shapeType=Capsule, length=0.505, radius=0.505, shapeName=hair_top_D_base_capsule, rotation=<rot 0 -90 0>}]
 hair_top_E_1 [{shapeType=Capsule, length=1, radius=1, shapeName=hair_top_E_1_capsule, center=<4.6183 0 0>}]
 hair_top_E_base [{shapeType=Capsule, length=0.505, radius=0.505, shapeName=hair_top_E_base_capsule, rotation=<rot 0 -90 0>}]
}
constraints [
 {bone1Name=hair_top_A_base, bone2Name=hair_top_A_1, swing1LimitDegrees=20, swing2LimitDegrees=20, twistLimitDegrees=5}
 {bone1Name=hair_top_A_1, bone2Name=hair_top_A_2, swing1LimitDegrees=20, swing2LimitDegrees=20, twistLimitDegrees=5}
 {bone1Name=hair_top_B_base, bone2Name=hair_top_B_1, swing1LimitDegrees=20, swing2LimitDegrees=20, twistLimitDegrees=5}
 {bone1Name=hair_top_B_1, bone2Name=hair_top_B_2, swing1LimitDegrees=20, swing2LimitDegrees=20, twistLimitDegrees=5}
 {bone1Name=hair_top_C_base, bone2Name=hair_top_C_1, swing1LimitDegrees=20, swing2LimitDegrees=20, twistLimitDegrees=5}
 {bone1Name=hair_top_C_1, bone2Name=hair_top_C_2, swing1LimitDegrees=20, swing2LimitDegrees=20, twistLimitDegrees=5}
 {bone1Name=hair_top_D_base, bone2Name=hair_top_D_1, swing1LimitDegrees=20, swing2LimitDegrees=20, twistLimitDegrees=5}
 {bone1Name=hair_top_D_1, bone2Name=hair_top_D_2, swing1LimitDegrees=20, swing2LimitDegrees=20, twistLimitDegrees=5}
 {bone1Name=hair_top_E_base, bone2Name=hair_top_E_1, swing1LimitDegrees=20, swing2LimitDegrees=20, twistLimitDegrees=5}
]
