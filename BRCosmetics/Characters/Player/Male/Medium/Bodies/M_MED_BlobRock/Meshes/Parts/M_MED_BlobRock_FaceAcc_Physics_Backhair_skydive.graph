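# Physics asset for the BlobRock character's back-hair face accessory, apparently the skydive variant.
# Left, middle and right hair chains each run from a MSTR bone through three capsule bodies with tight swing and twist limits.
# Head, upper spine, clavicle and upper-arm capsules provide body shapes for the hair to collide against.
# Angularly free cross-constraints tie left to middle at segment 2 and right to middle at segment 3, linking the locks.
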PhysicsAsset /BRCosmetics/Characters/Player/Male/Medium/Bodies/M_MED_BlobRock/Meshes/Parts/M_MED_BlobRock_FaceAcc_Physics_Backhair_skydive
bodies {
 clavicle_l [{
   shapeType=Capsule
   length=5.7622
   radius=12.2957
   shapeName=clavicle_l_capsule
   center=<13.0879 0.5433 -3.617>
   rotation=<rot 80 -0.0003 -0.0003>
  }]
 clavicle_r [{
   shapeType=Capsule
   length=4.529
   radius=13.5289
   shapeName=clavicle_r_capsule
   center=<-12.3907 -0.3448 5.292>
   rotation=<rot 70 -0.0004 -0.0005>
  }]
 hair_back_l_1 [{shapeType=Capsule, length=0.481, radius=3.3095, shapeName=hair_back_l_1_capsule, center=<3.3056 0 0.312>, rotation=<rot 90 0 0>}]
 hair_back_l_2 [
  {shapeType=Capsule, length=0, radius=5.9566, shapeName=hair_back_l_2_capsule, center=<3.3056 0.8096 1.669>, rotation=<rot 90 0 0>}
  {shapeType=Capsule, length=2.0131, radius=2.8716, shapeName=hair_back_l_2_capsule, center=<3.4191 4.1014 -2.8025>, rotation=<rot 90 0 0>}
 ]
 hair_back_l_3 [{shapeType=Capsule, length=0.922, radius=4.7356, shapeName=hair_back_l_3_capsule, center=<2.2471 -0.1459 -0.0272>, rotation=<rot 90 0 0>}]
 hair_back_l_MSTR [{shapeType=Capsule, length=1, radius=1, shapeName=hair_back_l_MSTR_capsule}]
 hair_back_mid_1 [{shapeType=Capsule, length=0.4643, radius=3.3402, shapeName=hair_back_mid_1_capsule, center=<3.2742 0 0.4978>, rotation=<rot 90 0 0>}]
 hair_back_mid_2 [{shapeType=Capsule, length=0.4186, radius=5.0505, shapeName=hair_back_mid_2_capsule, center=<3.3056 -0.3753 1.613>, rotation=<rot 90 0 0>}]
 hair_back_mid_3 [{shapeType=Capsule, length=1.5, radius=5.052, shapeName=hair_back_mid_3_capsule, center=<4.6305 -0.7256 3.2467>, rotation=<rot 90 0 0>}]
 hair_back_mid_MSTR [{shapeType=Capsule, length=1, radius=1, shapeName=hair_back_mid_MSTR_capsule}]
 hair_back_r_1 [{shapeType=Capsule, length=0.0953, radius=3.3801, shapeName=hair_back_r_1_capsule, center=<3.3056 0 0>, rotation=<rot 90 0 0>}]
 hair_back_r_2 [
  {
   shapeType=Capsule
   length=0
   radius=5.159
   shapeName=hair_back_r_2_capsule
   center=<3.3056 -1.9905 1.7049>
   rotation=<rot 60.0001 -89.9996 -89.9994>
  }
  {
   shapeType=Capsule
   length=2.4809
   radius=2.5179
   shapeName=hair_back_r_2_capsule
   center=<3.2853 -3.937 -2.79>
   rotation=<rot -72.8734 17.628 -33.9431>
  }
 ]
 hair_back_r_3 [{
   shapeType=Capsule
   length=4.156
   radius=5.7909
   shapeName=hair_back_r_3_capsule
   center=<3.8796 2.1714 -0.3417>
   rotation=<rot 79.9998 10.0004 0.0006>
  }]
 hair_back_r_MSTR [{shapeType=Capsule, length=1, radius=1, shapeName=hair_back_r_MSTR_capsule}]
 head [{shapeType=Capsule, length=2.2148, radius=6.9503, shapeName=head_capsule, center=<0.4302 3.9766 0>, rotation=<rot -90 0 0>}]
 spine_05 [
  {shapeType=Capsule, length=12.3512, radius=17.691, shapeName=spine_05_capsule, center=<-3.3664 3.2264 0>}
  {shapeType=Capsule, length=19.9701, radius=13.3529, shapeName=spine_05_capsule, center=<1.9223 -14.4568 0>, rotation=<rot 0 9.9998 0>}
 ]
 upperarm_l [{shapeType=Capsule, length=9.1834, radius=9.384, shapeName=upperarm_l_capsule, center=<7.8669 0 0>, rotation=<rot 90 0 0>}]
 upperarm_r [{shapeType=Capsule, length=9.8231, radius=8.7443, shapeName=upperarm_r_capsule, center=<-9.2459 0 0>, rotation=<rot 90 0 0>}]
}
constraints [
 {bone1Name=hair_back_l_3, bone2Name=hair_back_l_2, swing1LimitDegrees=20, swing2LimitDegrees=20, twistLimitDegrees=5}
 {bone1Name=hair_back_l_2, bone2Name=hair_back_l_1, swing1LimitDegrees=20, swing2LimitDegrees=20, twistLimitDegrees=5}
 {bone1Name=hair_back_l_1, bone2Name=hair_back_l_MSTR, swing1LimitDegrees=20, swing2LimitDegrees=40, twistLimitDegrees=5}
 {bone1Name=hair_back_r_3, bone2Name=hair_back_r_2, swing1LimitDegrees=20, swing2LimitDegrees=20, twistLimitDegrees=5}
 {bone1Name=hair_back_r_2, bone2Name=hair_back_r_1, swing1LimitDegrees=20, swing2LimitDegrees=20, twistLimitDegrees=20}
 {bone1Name=hair_back_r_1, bone2Name=hair_back_r_MSTR, swing1LimitDegrees=20, swing2LimitDegrees=40, twistLimitDegrees=5}
 {bone1Name=hair_back_mid_3, bone2Name=hair_back_mid_2, swing1LimitDegrees=20, swing2LimitDegrees=20, twistLimitDegrees=5}
 {bone1Name=hair_back_mid_2, bone2Name=hair_back_mid_1, swing1LimitDegrees=20, swing2LimitDegrees=20, twistLimitDegrees=5}
 {bone1Name=hair_back_mid_1, bone2Name=hair_back_mid_MSTR, swing1LimitDegrees=20, swing2LimitDegrees=40, twistLimitDegrees=5}
 {
  bone1Name=hair_back_l_2
  bone2Name=hair_back_mid_2
  swing1LimitDegrees=45
  swing1Motion=Free
  swing2LimitDegrees=45
  swing2Motion=Free
  twistLimitDegrees=45
  twistMotion=Free
 }
 {
  bone1Name=hair_back_r_3
  bone2Name=hair_back_mid_3
  swing1LimitDegrees=45
  swing1Motion=Free
  swing2LimitDegrees=45
  swing2Motion=Free
  twistLimitDegrees=45
  twistMotion=Free
 }
]
