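# Post-process style AnimBlueprint for the BlobRock face accessory that simulates hair physics on the character.
# It copies the parent mesh pose through a Control Rig and upper-hair RigidBody, weaker in air, into a cached pose.
# Back-hair RigidBody and Control Rig branches switch on backpack and skydiving state, then get a small bone offset.
# Events update state-based gravity and set backpack push, reduced for the matching BlobRock pack.
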AnimBlueprint /BRCosmetics/Characters/Player/Male/Medium/Bodies/M_MED_BlobRock/Meshes/Parts/M_MED_BlobRock_FaceAcc_AnimBP
target_skeleton=&/M_MED_BlobRock_FaceAcc_Skeleton
graphs [
 AnimGraph {nodes [
   n1 = RigidBody!(
    ComponentPose=Local To Component(LocalPose=Use cached pose 'main'())
    Alpha
    OverrideWorldGravity:Vector=Get Gravity Override()
    ExternalForce
   ) -> (Pose) {
    bOverrideWorldGravity=true
    overridePhysicsAsset=&/M_MED_BlobRock_FaceAcc_Physics_Backhair_skydive
    simulationSpace=BaseBoneSpace
    baseBoneRef {boneName=neck_01}
    componentAppliedLinearAccClamp=<0 0 0>
    simSpaceSettings {
     dampingAlpha=0
     maxAngularAcceleration=8
     maxAngularVelocity=8
     maxLinearAcceleration=150
     maxLinearVelocity=300
     worldAlpha=1
     externalLinearDragV=<3 3 3>
    }
   }
   n2 = Control Rig!(Source=Component To Local(ComponentPose=@n1), Alpha, is skydive:bool=false, backpack_push:double=Get hair_push() -> (hair_push:double)) -> (Pose) {controlRigAssetReference {blueprintRigClass=&/M_MED_BlobRock_FaceAcc_CtrlRig_BackHair.M_MED_BlobRock_FaceAcc_CtrlRig_BackHair_C}}
   n3 = RigidBody(ComponentPose=Local To Component(LocalPose=Use cached pose 'main'()), OverrideWorldGravity="0,0,-980") {
    bOverrideWorldGravity=true
    overridePhysicsAsset=&/M_MED_BlobRock_FaceAcc_Physics_BackHair_bakcpack
    simulationSpace=BaseBoneSpace
    baseBoneRef {boneName=neck_01}
    componentAppliedLinearAccClamp=<0 0 0>
    simSpaceSettings {dampingAlpha=0, maxAngularAcceleration=8, maxAngularVelocity=8, maxLinearAcceleration=0, maxLinearVelocity=0, worldAlpha=1}
   }
   n4 = Control Rig(Source=Component To Local(ComponentPose=@n3), is skydive=false, backpack_push=Get hair_push()) {controlRigAssetReference {blueprintRigClass=&M_MED_BlobRock_FaceAcc_CtrlRig_BackHair_C}}
   n5 = Blend Poses by bool(BlendPose_0=@n2, BlendPose_1=@n4, bActiveValue=Get bIsSkyDiving())
   n6 = RigidBody(ComponentPose=Local To Component(LocalPose=Use cached pose 'main'()), OverrideWorldGravity=Get Gravity Override()) {
    bOverrideWorldGravity=true
    overridePhysicsAsset=&/M_MED_BlobRock_FaceAcc_Physics_BackHair
    simulationSpace=BaseBoneSpace
    baseBoneRef {boneName=neck_01}
    componentAppliedLinearAccClamp=<0 0 0>
    simSpaceSettings {
     dampingAlpha=0
     maxAngularAcceleration=8
     maxAngularVelocity=8
     maxLinearAcceleration=200
     maxLinearVelocity=400
     worldAlpha=1
     externalLinearDragV=<3 3 3>
    }
   }
   n7 = Control Rig!(Source=Component To Local(ComponentPose=@n6), Alpha, backpack:bool=false, is skydive=Get bIsSkyDiving(), backpack_push=0) -> (Pose) {controlRigAssetReference {blueprintRigClass=&M_MED_BlobRock_FaceAcc_CtrlRig_BackHair_C}}
   n8 = Get is_wear_backpack() -> (is_wear_backpack:bool)
   n9 = "Transform (Modify) Bone - Bone: hair_top_B_base"(
    ComponentPose=Local To Component(LocalPose=Blend Poses by bool(BlendPose_0=@n5, BlendPose_1=@n7, bActiveValue=@n8))
    Alpha=1
    Translation="0,0.3,0.3"
    Rotation="0,0,0"
    Scale="1,1,1"
   ) {translationMode=BMM_Additive, translationSpace=BCS_BoneSpace, boneToModify {boneName=hair_top_B_base}, translation {y=0.3, z=0.3}}
   Output Pose(Result=Component To Local(ComponentPose=@n9)) {layerGroup=None}
   n10 = Control Rig!(Source=Copy Pose From Mesh() {bUseAttachedParent=true}, Alpha) -> (Pose) {controlRigAssetReference {blueprintRigClass=&/M_MED_BlobRock_FaceAcc_CtrlRig.M_MED_BlobRock_FaceAcc_CtrlRig_C}}
   n11 = RigidBody(
    ComponentPose=Local To Component(LocalPose=@n10)
    Alpha=Get up_hair_dyn() -> (up_hair_dyn:double)
    OverrideWorldGravity=Get Gravity Override()
   ) {
    bOverrideWorldGravity=true
    overridePhysicsAsset=&/M_MED_BlobRock_FaceAcc_Physics
    simulationSpace=BaseBoneSpace
    baseBoneRef {boneName=neck_01}
    componentAppliedLinearAccClamp=<0 0 0>
    simSpaceSettings {
     dampingAlpha=0
     maxAngularAcceleration=5
     maxAngularVelocity=5
     maxLinearAcceleration=40
     maxLinearVelocity=70
     worldAlpha=1
     externalLinearDragV=<2 2 2>
    }
   }
   Save cached pose 'main'(Pose=Component To Local(ComponentPose=@n11)) {cacheName=main}
  ]}
 EventGraph {nodes [
   Event BlueprintUpdateAnimation()
   > n1 = Gravity Override State Function DEPRECATED(
    CharacterPartAnimInstance=Self-Reference()
    GameTimeInSeconds=GetGameTimeInSeconds()
    GravityOverrideParams=Get Gravity Override Params()
    RBAN Params=Get RBAN Params()
   )
   > Set Gravity Override(Gravity Override=@n1)
   n2 = Get bIsInAir() -> (bIsInAir:bool)
   > n3 = Branch(Condition=@n2)
   > Set up_hair_dyn(up_hair_dyn:double=0.3) -> (Output_Get:double)
   Set up_hair_dyn(execute=@n3.else, up_hair_dyn=1)
   n4 = Event OnCharacterCustomizationLoaded()
   > Set is_wear_backpack(is_wear_backpack:bool=CheckForBackbling()) -> (Output_Get:bool)
   n5 = Get Array Item with EFort Custom Part Type(Array=@n4.CustomMeshParts, PartType=Backpack)
   > n6 = Branch(Condition="Equal (Object)"(
     A=Get SkeletalMesh!(self:Skinned Mesh Component Object Reference=@n5) -> (SkeletalMesh:Skeletal Mesh Object Reference)
     B=/BRCosmetics/Accessories/FORT_Backpacks/Backpack_M_MED_BlobRock/Meshes/M_MED_BlobRock_Pack
    ))
   > Set hair_push(hair_push:double=0.5) -> (Output_Get:double)
   Set hair_push(execute=@n6.else, hair_push=1)
   OR Boolean(A=@n2, B=Get bIsTacticalSprinting())
  ]}
]
variables [
 backpack {type=SkeletalMeshComponentBudgeted}
 Gravity Override {type=Vector}
 Gravity Override Params {type=GravityOverrideParamsStruct}
 hair_push {type=number}
 is_wear_backpack {type=boolean}
 RBAN Params {
  type=RBANParamsStruct
  default {
   dBNO {simSpaceSettings {
     maxAngularAcceleration=1000
     maxAngularVelocity=3
     maxLinearAcceleration=300
     maxLinearVelocity=700
     velocityScaleZ=0
     externalLinearDragV=<0 0.5 0>
    }}
   dBNO_Carried {simSpaceSettings {
     maxAngularAcceleration=1000
     maxAngularVelocity=3
     maxLinearAcceleration=300
     maxLinearVelocity=700
     velocityScaleZ=0
     externalLinearDragV=<0 0.5 0>
    }}
   emoteOrMelee {simSpaceSettings {
     maxAngularAcceleration=1000
     maxAngularVelocity=3
     maxLinearAcceleration=300
     maxLinearVelocity=700
     velocityScaleZ=0
     externalLinearDragV=<0 0.5 0>
    }}
   front_End {simSpaceSettings {
     maxAngularAcceleration=1000
     maxAngularVelocity=3
     maxLinearAcceleration=300
     maxLinearVelocity=700
     velocityScaleZ=0
     externalLinearDragV=<0 0.5 0>
    }}
   onGround_Crouch_Moving {simSpaceSettings {
     maxAngularAcceleration=1000
     maxAngularVelocity=3
     maxLinearAcceleration=300
     maxLinearVelocity=700
     velocityScaleZ=0
     externalLinearDragV=<0 0.5 0>
    }}
   onGround_Falling {simSpaceSettings {
     maxAngularAcceleration=1000
     maxAngularVelocity=3
     maxLinearAcceleration=300
     maxLinearVelocity=700
     velocityScaleZ=0
     externalLinearDragV=<0 0.5 0>
    }}
   onGround_Moving {simSpaceSettings {
     maxAngularAcceleration=1000
     maxAngularVelocity=3
     maxLinearAcceleration=300
     maxLinearVelocity=700
     velocityScaleZ=0
     externalLinearDragV=<0 0.5 0>
    }}
   onGround_Standing {simSpaceSettings {
     maxAngularAcceleration=1000
     maxAngularVelocity=3
     maxLinearAcceleration=300
     maxLinearVelocity=700
     velocityScaleZ=0
     externalLinearDragV=<0 0.5 0>
    }}
   rail_Grinding {simSpaceSettings {
     maxAngularAcceleration=1000
     maxAngularVelocity=3
     maxLinearAcceleration=300
     maxLinearVelocity=700
     velocityScaleZ=0
     worldAlpha=0.1
     externalLinearDragV=<0 0.5 0>
    }}
   riding_Moving {simSpaceSettings {
     maxAngularAcceleration=1000
     maxAngularVelocity=3
     maxLinearAcceleration=300
     maxLinearVelocity=700
     velocityScaleZ=0
     externalLinearDragV=<0 0.5 0>
    }}
   riding_Moving_Targeting {simSpaceSettings {
     maxAngularAcceleration=1000
     maxAngularVelocity=3
     maxLinearAcceleration=300
     maxLinearVelocity=700
     velocityScaleZ=0
     externalLinearDragV=<0 0.5 0>
    }}
   riding_Standing {simSpaceSettings {
     maxAngularAcceleration=1000
     maxAngularVelocity=3
     maxLinearAcceleration=300
     maxLinearVelocity=700
     velocityScaleZ=0
     externalLinearDragV=<0 0.5 0>
    }}
   skydive_DivingDown {simSpaceSettings {
     maxAngularAcceleration=1000
     maxAngularVelocity=3
     maxLinearAcceleration=300
     maxLinearVelocity=700
     velocityScaleZ=0
     externalLinearDragV=<0 0.5 0>
    }}
   skydive_FloatingDown {
    componentLinearAccScale=<0 0.5 0>
    componentLinearVelScale=<0 0.5 0>
    simSpaceSettings {
     maxAngularAcceleration=1000
     maxAngularVelocity=3
     maxLinearAcceleration=300
     maxLinearVelocity=700
     velocityScaleZ=0
     externalLinearDragV=<0 0.5 0>
    }
   }
   skydive_GoingUp {simSpaceSettings {
     maxAngularAcceleration=1000
     maxAngularVelocity=3
     maxLinearAcceleration=300
     maxLinearVelocity=700
     velocityScaleZ=0
     externalLinearDragV=<0 0.5 0>
    }}
   skydive_Parachuting {
    componentLinearAccScale=<0 0.1 0.3>
    componentLinearVelScale=<0 0.1 0.3>
    simSpaceSettings {
     maxAngularAcceleration=1000
     maxAngularVelocity=3
     maxLinearAcceleration=300
     maxLinearVelocity=700
     velocityScaleZ=0
     externalLinearDragV=<0 0.5 0>
    }
   }
   skydive_Surfing {simSpaceSettings {
     maxAngularAcceleration=1000
     maxAngularVelocity=3
     maxLinearAcceleration=300
     maxLinearVelocity=700
     velocityScaleZ=0
     externalLinearDragV=<0 0.5 0>
    }}
   swimming {simSpaceSettings {
     maxAngularAcceleration=1000
     maxAngularVelocity=3
     maxLinearAcceleration=300
     maxLinearVelocity=700
     velocityScaleZ=0
     externalLinearDragV=<0 0.5 0>
    }}
   tactical_Sprint {simSpaceSettings {
     maxAngularAcceleration=1000
     maxAngularVelocity=3
     maxLinearAcceleration=300
     maxLinearVelocity=700
     velocityScaleZ=0
     externalLinearDragV=<0 0.5 0>
    }}
   ziplining {simSpaceSettings {
     maxAngularAcceleration=1000
     maxAngularVelocity=3
     maxLinearAcceleration=300
     maxLinearVelocity=700
     velocityScaleZ=0
     externalLinearDragV=<0 0.5 0>
    }}
  }
 }
 up_hair_dyn {type=number}
]
